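; HUD view model that shows the player's tactical sprint energy on a progress bar while the Stalker Knife is equipped.
; On initialize it caches the pawn's TacticalSprintEnergyComponent, sets a knife-specific bar material and listens for weapon-equipped events.
; A weapon class check, standing in for a planned tag check, enables the bar and binds energy changes, or disables, zeroes and unbinds it.
; Uninitialize unbinds the pawn's unequipped delegate, not the equipped one it bound.
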(event EventInitialize
  (params (Pawn :type Pawn%20Object%20Reference))
  (|CallParentFunction Pawn)
  (Utilities|IsValid Pawn
    (:Is%20Valid
      (Variables|Default|SetTacticalSprintComponent (Actor|GetComponentbyClass Pawn "/TacticalSprintGame/Gameplay/TacticalSprintEnergyComponent.TacticalSprintEnergyComponent_C"))
      ;[1 0.03562 0.7265 0.5] Listening to when Stalker Knife is equipped
      (bind _asfort_pawn (Utilities|Casting|CastToFortPawn Pawn))
      (FortPawn|BindEventtoOnWeaponEquippedDelegate _asfort_pawn (EventDispatchers|CreateEvent :!event Custom|OnWeaponChanged))
      (CallFunction|SetProgressBarOverrideMaterial "/FNMStalkerKnife/UI/Materials/MI_UI_TacSprintStalkerKnife_Rounded.MI_UI_TacSprintStalkerKnife_Rounded"))))

(event EventUninitialize
  (params (Pawn :type Pawn%20Object%20Reference))
  (bind _tacticalsprintcomponent (Variables|Default|GetTacticalSprintComponent))
  (Utilities|IsValid _tacticalsprintcomponent
    (:Is%20Valid
      (Events|UnbindEventfromOnCurrentEnergyChanged _tacticalsprintcomponent (EventDispatchers|CreateEvent :!event OnCurrentEnergyChanged))
      (bind _asfort_pawn (Utilities|Casting|CastToFortPawn Pawn))
      (FortPawn|UnbindallEventsfromOnWeaponUnEquippedDelegate _asfort_pawn))))

(event Custom|OnWeaponChanged
  (params (NewWeapon :type Fort%20Weapon%20Object%20Reference) (PrevWeapon :type Fort%20Weapon%20Object%20Reference))
  ; TODO: change this to check the Equipped tag instead. At the moment it's not possible because the tag is not being added to the pawn.
  (CallFunction|ToggleStalkerKnifeEnergyBar (== (Utilities|GetClass NewWeapon) "/FNMStalkerKnife/Gameplay/B_StalkerKnife_MantisWeapon.B_StalkerKnife_MantisWeapon_C")))

(fn CreateEnergyChangedBinding
  (bind _tacticalsprintcomponent (Variables|Default|GetTacticalSprintComponent))
  (Utilities|IsValid _tacticalsprintcomponent
    (:Is%20Valid
      (CallFunction|OnCurrentEnergyChanged _tacticalsprintcomponent)
      (Events|BindEventtoOnCurrentEnergyChanged _tacticalsprintcomponent (EventDispatchers|CreateEvent :!event OnCurrentEnergyChanged)))))

(fn OnCurrentEnergyChanged
  (params (EnergyComponent :type Fort%20Component%20Energy%20Object%20Reference) (PreviousValue :type Float%20%28double-precision%29))
  (Variables|GenericProgressBarViewModel|SetProgressValue (Energy|GetCurrentEnergyPercentage EnergyComponent)))

(fn RemoveEnergyChangedBinding
  (bind _tacticalsprintcomponent (Variables|Default|GetTacticalSprintComponent))
  (Utilities|IsValid _tacticalsprintcomponent
    (:Is%20Valid
      (Events|UnbindEventfromOnCurrentEnergyChanged _tacticalsprintcomponent (EventDispatchers|CreateEvent :!event OnCurrentEnergyChanged))
      (Variables|GenericProgressBarViewModel|SetProgressValue 0.0))))

(fn ShouldBecomeVisibleImmediately
  (bind _tacticalsprintcomponent (Variables|Default|GetTacticalSprintComponent))
  (Utilities|IsValid _tacticalsprintcomponent
    (:Is%20Valid
      (return 0.0 (< (Energy|GetCurrentEnergyPercentage _tacticalsprintcomponent) 1.0)))
    (:Is%20Not%20Valid
      (return 0.0 false))))

(fn ToggleStalkerKnifeEnergyBar
  (params (bIsActive :type Boolean))
  (if bIsActive
    (CallFunction|CreateEnergyChangedBinding)
    (Variables|GenericProgressBarViewModel|SetbEnabled true)
    (else
      (CallFunction|RemoveEnergyChangedBinding)
      (Variables|GenericProgressBarViewModel|SetbEnabled false))))
